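; Big rig vehicle spawner device Blueprint, a thin child that inherits its spawning behavior from a parent class.
; It overrides the visibility component lists to cover the base static mesh and a plate extension, both mapped to QueryAndPhysics collision.
; In the editor, HandleEditorVisibility sets the base mesh visibility from bIsVisible without propagating to children.
; The construction script only calls the parent, and the event graph holds no active logic.
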(event EventBeginPlay
  (|CallParentFunction) ; This node is disabled and will not be called.\nDrag off pins to build functionality.
)

(event Collision|EventActorBeginOverlap
  (params (OtherActor :type Actor%20Object%20Reference)))

(event EventTick
  (params (DeltaSeconds :type Float%20%28single-precision%29)))

(fn GetVisibilityComponents
  (return (Utilities|Array|MakeArray (Variables|Debug|GetStaticMeshComponent) (Variables|Default|GetPlateExtension))))

(fn GetVisibilityComponentsWithCollisionOverrides
  (return (Utilities|Map|MakeMap (Variables|Debug|GetStaticMeshComponent) "QueryAndPhysics" (Variables|Default|GetPlateExtension) "QueryAndPhysics")))

(fn HandleEditorVisibility
  ;[0.8827 0.1953 1 0.25] If we're the editor, hide the base if we don't want to show.
  ; (since we don't need to interact with it and we have the preview)
  (if (Editor|InEditor)
    (Rendering|SetVisibility (Variables|Debug|GetStaticMeshComponent) (Variables|LegacyOptions|GetbIsVisible) false)))

(fn ConstructionScript
  (|CallParentFunction))
