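; Creative device Blueprint for the Big Rig vehicle spawner, layering color/style options and base-plate visibility over parent spawner logic.
; Color and Style converts to and from an index offset by one, with a special-case remap because the added black option breaks one-to-one enum mapping.
; HandleEditorVisibility sets base mesh visibility in editor; otherwise it shows the plate extension outside play mode or when Visible During Game is set.
; Both meshes are reported as visibility components with QueryAndPhysics collision overrides.
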(event EventBeginPlay
  (|CallParentFunction) ; This node is disabled and will not be called.\nDrag off pins to build functionality.
)

(event Collision|EventActorBeginOverlap
  (params (OtherActor :type Actor%20Object%20Reference)))

(event EventTick
  (params (DeltaSeconds :type Float%20%28single-precision%29)))

(fn ColorAndStyleAsAnInteger
  (return (- (Math|Conversions|ToInteger(Byte) (Variables|UserOptions|Basic|GetColor%20and%20Style)) 1)))

(fn GetVisibilityComponents
  (return (Utilities|Array|MakeArray (Variables|Debug|GetStaticMeshComponent) (Variables|Default|GetPlateExtension))))

(fn GetVisibilityComponentsWithCollisionOverrides
  (return (Utilities|Map|MakeMap (Variables|Debug|GetStaticMeshComponent) "QueryAndPhysics" (Variables|Default|GetPlateExtension) "QueryAndPhysics")))

(fn HandleEditorVisibility
  (bind _fortminigameprogress (Variables|Default|GetFortMinigameProgress))
  (bind _play_mode (Class|CreativeCommonDeviceFunctionLibrary|IsLogicinPlayMode _fortminigameprogress))
  (bind _returnvalue (not _play_mode))
  (bind _visible_during_game (Variables|UserOptions|Basic|GetVisible%20During%20Game))
  (bind _returnvalue_2 (and _play_mode _visible_during_game))
  (bind _returnvalue_3 (or _returnvalue _returnvalue_2))
  (bind _plateextension (Variables|Default|GetPlateExtension))
  ;[0.8827 0.1953 1 0.25] If we're the editor, hide the base if we don't want to show.
  ; (since we don't need to interact with it and we have the preview)
  (if (Editor|InEditor)
    (Rendering|SetVisibility (Variables|Debug|GetStaticMeshComponent) (Variables|UserOptions|Basic|GetVisible%20During%20Game) false)
    (else
      ;[1 1 1] For reasons I can not understand this is affecting the visibility of the spawn plates in game if you go from editor to game
      (Rendering|SetVisibility _plateextension _returnvalue_3 false)
      (Development|SetHiddeninGame _plateextension (not _returnvalue_3) false))))

(fn SetColorOption
  (params (ColorIndex :type Integer))
  ;[1 1 1] Because Graffiti is actually Mat Index 7, and we are adding black as an option, we can't go 1:1 on all the values.
  (Variables|UserOptions|Basic|SetColor%20and%20Style (select (== (+ ColorIndex 1) 1) "NewEnumerator27" (Utilities|Enum|BytetoEnumEColorBigRig (Math|Conversions|ToByte(Integer) (+ ColorIndex 1))))))

(fn ConstructionScript
  (|CallParentFunction))
